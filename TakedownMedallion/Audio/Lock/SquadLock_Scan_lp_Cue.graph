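# Looping Sound Cue for the SquadLock scan sound in the Takedown Medallion content.
# A single looping wave player plays the scan wave at reduced pitch and volume, using the ambient props sound class and SFX submix.
# Overridden attenuation applies natural-sound falloff, distance-based low-pass filtering and stereo spread.
# Overridden concurrency caps it at two instances and blocks new ones beyond that.
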SoundCue /TakedownMedallion/Audio/Lock/SquadLock_Scan_lp_Cue
root=WavePlayer_1
nodes [{id=WavePlayer_1, properties {bLooping=true, soundWaveAssetPtr=/TakedownMedallion/Audio/Lock/wav/Squadlocks_Scan_lp}}]
properties {
 attenuationSettings=/Game/Sounds/Attenuation/Fort_Amb_Emitter_Prop_Med_Att
 bOverrideAttenuation=true
 bOverrideConcurrency=true
 maxDistance=2700
 pitchMultiplier=0.8
 soundClassObject=&Fort_Ambient_Props_Athena
 soundSubmixObject=&Fortnite_SFX
 volumeMultiplier=0.8
 attenuationOverrides {
  bAttenuateWithLPF=true
  distanceAlgorithm=NaturalSound
  falloffDistance=2200
  lPFFrequencyAtMax=2000
  lPFFrequencyAtMin=10000
  lPFRadiusMax=1500
  lPFRadiusMin=500
  nonSpatializedRadiusStart=10
  occlusionLowPassFilterFrequency=6000
  reverbDistanceMax=2000
  stereoSpread=550
  attenuationShapeExtents {x=500}
 }
 concurrencyOverrides {maxCount=2, resolutionRule=PreventNew}
}
