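# SquadLock_Unlock_Cue is the sound cue for a squad lock unlocking in the Takedown Medallion content.
# A fixed short Delay feeds a single Wave Player that plays the Squadlocks_Unlock wave.
# It routes to the gadgets sound class and SFX submix, with overridden natural-sound attenuation for spatialized falloff.
# A concurrency override limits it to one playing instance at a time.
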SoundCue /TakedownMedallion/Audio/Lock/SquadLock_Unlock_Cue
root=Delay_0
nodes [
 {id=Delay_0, children [WavePlayer_1], properties {delayMax=0.3, delayMin=0.3}}
 {id=WavePlayer_1, properties {soundWaveAssetPtr=/TakedownMedallion/Audio/Lock/wav/Squadlocks_Unlock}}
]
properties {
 attenuationSettings=/Game/Sounds/Attenuation/Fort_Explosion_Medium_StW_Att
 bOverrideAttenuation=true
 bOverrideConcurrency=true
 maxDistance=4050
 soundClassObject=&Fort_Gadgets
 soundSubmixObject=&Fortnite_SFX
 volumeMultiplier=0.798
 attenuationOverrides {distanceAlgorithm=NaturalSound, nonSpatializedRadiusStart=60, stereoSpread=800, attenuationShapeExtents {x=450}}
 concurrencyOverrides {maxCount=1}
}
